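# Physics asset for the face accessory part of the female medium Duck Hero character, defining simulation for its curl and eyelash bones.
# Box bodies cover the head, two curl chains and individual left and right lash bones.
# Each lash and each curl root is constrained to the head, and later curl segments link to the previous one.
# Swing and twist limits are tightest at curl roots and loosest down the chains.
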PhysicsAsset /BRCosmetics/Characters/Player/Female/Medium/Bodies/F_MED_Duck_Hero/Meshes/Parts/F_MED_DuckHero_FaceAcc_Physics
bodies {
 dyn_curlA_01 [{shapeType=Box, extentX=2.9649, extentY=3.3051, extentZ=4.7973, center=<1.2 0 0>, rotation=<rot 90 0 0>}]
 dyn_curlA_02 [{shapeType=Box, extentX=2.434, extentY=2.9082, extentZ=4.2377, center=<1.2 0 0>, rotation=<rot 90 0 0>}]
 dyn_curlA_03 [{shapeType=Box, extentX=2.4316, extentY=2.2849, extentZ=3.477, center=<1.2 0 0>, rotation=<rot 90 0 0>}]
 dyn_curlB_01 [{shapeType=Box, extentX=2.0754, extentY=1.6004, extentZ=4.0981, center=<1.2 0 0>, rotation=<rot 90 0 0>}]
 dyn_curlB_02 [{shapeType=Box, extentX=1.8193, extentY=1.1779, extentZ=4.4152, center=<1.2 0 0>, rotation=<rot 90 0 0>}]
 dyn_lashA_01_l [{shapeType=Box, extentX=2, extentY=2, extentZ=2, center=<-0.866 0.313 -0.8931>, rotation=<rot 90 0 0>}]
 dyn_lashA_01_r [{shapeType=Box, extentX=2, extentY=2, extentZ=2, center=<-0.7968 0.4351 0.9063>, rotation=<rot 90 0 0>}]
 dyn_lashB_01_l [{shapeType=Box, extentX=2, extentY=2, extentZ=2, center=<-1.1529 -0.1704 -0.5361>, rotation=<rot 90 0 0>}]
 dyn_lashB_01_r [{shapeType=Box, extentX=2, extentY=2, extentZ=2, center=<-1.0111 0.4567 0.6439>, rotation=<rot 90 0 0>}]
 dyn_lashC_01_l [{shapeType=Box, extentX=2, extentY=2, extentZ=2, center=<-1.1417 0.2245 0.54>, rotation=<rot 90 0 0>}]
 dyn_lashC_01_r [{shapeType=Box, extentX=2, extentY=2, extentZ=2, center=<-1.1487 0.5526 0.1434>, rotation=<rot 90 0 0>}]
 dyn_lashD_01_r [{shapeType=Box, extentX=2, extentY=2, extentZ=2, center=<-1.0942 0.6472 -0.1718>, rotation=<rot 90 0 0>}]
 head [{shapeType=Box, extentX=5.6586, extentY=5.4473, extentZ=9.9899}]
}
constraints [
 {bone1Name=dyn_curlA_01, bone2Name=head, swing1LimitDegrees=6, swing2LimitDegrees=6, twistLimitDegrees=4}
 {bone1Name=dyn_curlA_02, bone2Name=dyn_curlA_01, swing1LimitDegrees=18, swing2LimitDegrees=12, twistLimitDegrees=8}
 {bone1Name=dyn_curlA_03, bone2Name=dyn_curlA_02, swing1LimitDegrees=18, swing2LimitDegrees=12, twistLimitDegrees=8}
 {bone1Name=dyn_curlB_01, bone2Name=head, swing1LimitDegrees=6, swing2LimitDegrees=6, twistLimitDegrees=4}
 {bone1Name=dyn_curlB_02, bone2Name=dyn_curlB_01, swing1LimitDegrees=18, swing2LimitDegrees=12, twistLimitDegrees=8}
 {bone1Name=dyn_lashA_01_r, bone2Name=head, swing1LimitDegrees=12, swing2LimitDegrees=12, twistLimitDegrees=8}
 {bone1Name=dyn_lashB_01_r, bone2Name=head, swing1LimitDegrees=12, swing2LimitDegrees=12, twistLimitDegrees=8}
 {bone1Name=dyn_lashC_01_r, bone2Name=head, swing1LimitDegrees=12, swing2LimitDegrees=12, twistLimitDegrees=8}
 {bone1Name=dyn_lashD_01_r, bone2Name=head, swing1LimitDegrees=12, swing2LimitDegrees=12, twistLimitDegrees=8}
 {bone1Name=dyn_lashC_01_l, bone2Name=head, swing1LimitDegrees=12, swing2LimitDegrees=12, twistLimitDegrees=8}
 {bone1Name=dyn_lashB_01_l, bone2Name=head, swing1LimitDegrees=12, swing2LimitDegrees=12, twistLimitDegrees=8}
 {bone1Name=dyn_lashA_01_l, bone2Name=head, swing1LimitDegrees=12, swing2LimitDegrees=12, twistLimitDegrees=8}
]
